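# Sound cue for the charge-up alarm of the DarkNexus StreetInfoBoard A prop's "SpecialA" sound.
# A single alarm wave plays through a Modulator left at its default settings, with the cue volume boosted.
# It uses special prop-interact attenuation and the in-game special item sound class.
# It also sends to two Fort ambient volume sidechain buses, one for outdoors, which suggests it ducks ambience while playing.
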SoundCue /Hera/Props/DarkNexus/DarkNexus_StreetInfoBoard_A/Audio/SC_DarkNexus_StreetInfoBoard_SpecialA_Alarm
root=Modulator_0
nodes [
 {id=Modulator_0, children [WavePlayer_0]}
 {
  id=WavePlayer_0
  properties {soundWaveAssetPtr=&/DarkNexus_StreetInfoBoard_SpecialA_Chargeup_Alarm_01}
 }
]
properties {
 attenuationSettings=/Game/Sounds/Attenuation/ATT_Fort_Prop_Interact_Special
 maxDistance=3500
 soundClassObject=/Game/Sounds/FortSoundClassesAndModes/Fort_LTM_BR_InGame_Special_Item
 volumeMultiplier=4
 busSends [
  {audioBus=&AB_Sidechain_FortAmbient_Vol, minSendLevel=0, sendLevel=0.5, soundSourceBus=null}
  {
   audioBus=/Game/Sounds/Modulation/EnvelopeFollowers/Volume/AB_SideChain_FortAmbient_Outdoors_Vol
   minSendLevel=0
   sendLevel=0.5
   soundSourceBus=null
  }
 ]
}
