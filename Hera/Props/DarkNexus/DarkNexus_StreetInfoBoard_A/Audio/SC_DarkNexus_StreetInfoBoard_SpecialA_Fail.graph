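# Sound cue for the failure sound of the DarkNexus street info board prop's SpecialA interaction.
# A Modulator feeds from a Random node that picks one of three equally weighted fail wave variants.
# It plays louder than default as a special in-game item sound, using special prop-interact attenuation and a concurrency set that, by its name, is shared with the success sound.
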SoundCue /Hera/Props/DarkNexus/DarkNexus_StreetInfoBoard_A/Audio/SC_DarkNexus_StreetInfoBoard_SpecialA_Fail
root=Modulator_0
nodes [
 {id=Modulator_0, children [Random_0]}
 {id=Random_0, children [WavePlayer_0,WavePlayer_1,WavePlayer_2], properties {weights [1,1,1]}}
 {
  id=WavePlayer_0
  properties {soundWaveAssetPtr=&/DarkNexus_StreetInfoBoard_SpecialA_Fail_01}
 }
 {
  id=WavePlayer_1
  properties {soundWaveAssetPtr=&/DarkNexus_StreetInfoBoard_SpecialA_Fail_02}
 }
 {
  id=WavePlayer_2
  properties {soundWaveAssetPtr=&/DarkNexus_StreetInfoBoard_SpecialA_Fail_03}
 }
]
properties {
 attenuationSettings=/Game/Sounds/Attenuation/ATT_Fort_Prop_Interact_Special
 maxDistance=3500
 priority=2
 soundClassObject=/Game/Sounds/FortSoundClassesAndModes/Fort_LTM_BR_InGame_Special_Item
 volumeMultiplier=3
 concurrencySet [&/CONC_DarkNexus_StreetInfoBoard_SpecialA_FailSuccess]
}
