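# Sound cue for the DarkNexus Street Info Board A prop, named as the start sound of its "Special A" event.
# A single wave player feeds a modulator left at default settings, which is the cue's root.
# The cue boosts volume and uses the shared special prop-interact attenuation and the in-game special item sound class.
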SoundCue /Hera/Props/DarkNexus/DarkNexus_StreetInfoBoard_A/Audio/SC_DarkNexus_StreetInfoBoard_SpecialA_Start
root=Modulator_0
nodes [
 {id=Modulator_0, children [WavePlayer_1]}
 {
  id=WavePlayer_1
  properties {soundWaveAssetPtr=&/DarkNexus_StreetInfoBoard_SpecialA_Start_01}
 }
]
properties {
 attenuationSettings=/Game/Sounds/Attenuation/ATT_Fort_Prop_Interact_Special
 maxDistance=3500
 soundClassObject=/Game/Sounds/FortSoundClassesAndModes/Fort_LTM_BR_InGame_Special_Item
 volumeMultiplier=2
}
